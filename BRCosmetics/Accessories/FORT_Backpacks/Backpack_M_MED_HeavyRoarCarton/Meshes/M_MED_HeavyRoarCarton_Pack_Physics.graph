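# Physics asset for the HeavyRoarCarton backpack cosmetic, giving secondary motion to its box, sp and meatball bones.
# Tiny capsules on base_ bones anchor capsule bodies on dyn_ bones through swing and twist limited constraints.
# Box parts swing mainly on one axis, meatballs are held to one degree, and sp strands swing more widely.
# Strands sp03–sp05 form chains, while sp06 branches, with both sp06_02 and sp06_03 constrained to dyn_sp06.
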PhysicsAsset /BRCosmetics/Accessories/FORT_Backpacks/Backpack_M_MED_HeavyRoarCarton/Meshes/M_MED_HeavyRoarCarton_Pack_Physics
bodies {
 base_box_L [{shapeType=Capsule, length=0.505, radius=0.505, shapeName=base_box_L_capsule, rotation=<rot 90 0 0>}]
 base_box_R [{shapeType=Capsule, length=0.505, radius=0.505, shapeName=base_box_R_capsule, rotation=<rot 90 0 0>}]
 base_box_back [{shapeType=Capsule, length=0.505, radius=0.505, shapeName=base_box_back_capsule, rotation=<rot 90 0 0>}]
 base_box_front [{shapeType=Capsule, length=0.505, radius=0.505, shapeName=base_box_front_capsule, rotation=<rot 90 0 0>}]
 base_meatball01 [{shapeType=Capsule, length=0.505, radius=0.505, shapeName=base_meatball01_capsule, rotation=<rot 0 -90 0>}]
 base_meatball02 [{shapeType=Capsule, length=0.505, radius=0.505, shapeName=base_meatball02_capsule, rotation=<rot 0 -90 0>}]
 base_sp01 [{shapeType=Capsule, length=0.505, radius=0.505, shapeName=base_sp01_capsule, rotation=<rot 90 0 0>}]
 base_sp02 [{shapeType=Capsule, length=0.505, radius=0.505, shapeName=base_sp02_capsule, rotation=<rot 90 0 0>}]
 base_sp03 [{shapeType=Capsule, length=0.505, radius=0.505, shapeName=base_sp03_capsule, rotation=<rot 90 0 0>}]
 base_sp04 [{shapeType=Capsule, length=0.505, radius=0.505, shapeName=base_sp04_capsule, rotation=<rot 90 0 0>}]
 base_sp05 [{shapeType=Capsule, length=0.505, radius=0.505, shapeName=base_sp05_capsule, rotation=<rot 0 -90 0>}]
 base_sp06 [{shapeType=Capsule, length=0.505, radius=0.505, shapeName=base_sp06_capsule, rotation=<rot 90 0 0>}]
 dyn_box_L [{shapeType=Capsule, length=2, radius=1, shapeName=dyn_box_L_capsule, center=<1.9432 0 0>, rotation=<rot 90 0 0>}]
 dyn_box_R [{shapeType=Capsule, length=2, radius=1, shapeName=dyn_box_R_capsule, center=<1.9432 0 0>, rotation=<rot 90 0 0>}]
 dyn_box_back [{shapeType=Capsule, length=2, radius=1, shapeName=dyn_box_back_capsule, center=<1.9432 0 0>, rotation=<rot 90 0 0>}]
 dyn_box_front [{shapeType=Capsule, length=8, radius=4.75, shapeName=dyn_box_front_capsule, center=<4.0148 0.0868 -4.0031>, rotation=<rot 0 0 -90>}]
 dyn_meatball01 [{shapeType=Capsule, length=1, radius=1, shapeName=dyn_meatball01_capsule}]
 dyn_meatball02 [{shapeType=Capsule, length=1, radius=1, shapeName=dyn_meatball02_capsule}]
 dyn_sp01 [{shapeType=Capsule, length=2, radius=1.5, shapeName=dyn_sp01_capsule, center=<1.6202 0 0.6034>, rotation=<rot 90 0 0>}]
 dyn_sp02 [{shapeType=Capsule, length=1, radius=1, shapeName=dyn_sp02_capsule, center=<0.9959 0 0>, rotation=<rot 90 0 0>}]
 dyn_sp03 [{shapeType=Capsule, length=2, radius=0.4, shapeName=dyn_sp03_capsule, center=<1.3865 0 0>, rotation=<rot 90 0 0>}]
 dyn_sp03_02 [{shapeType=Capsule, length=2, radius=0.4, shapeName=dyn_sp03_02_capsule, center=<1 0 0>, rotation=<rot 90 0 0>}]
 dyn_sp04 [{shapeType=Capsule, length=2, radius=0.4, shapeName=dyn_sp04_capsule, center=<1.3865 0 0>, rotation=<rot 90 0 0>}]
 dyn_sp04_02 [{shapeType=Capsule, length=2, radius=0.4, shapeName=dyn_sp04_02_capsule, center=<1.9337 -0.2945 0>, rotation=<rot 90 0 8.4374>}]
 dyn_sp04_03 [{shapeType=Capsule, length=2, radius=0.4, shapeName=dyn_sp04_03_capsule, center=<1.3865 0 0>, rotation=<rot 90 0 0>}]
 dyn_sp05 [{shapeType=Capsule, length=2, radius=0.4, shapeName=dyn_sp05_capsule, center=<1.1257 -0.0595 0>, rotation=<rot 90 0 8.4375>}]
 dyn_sp05_02 [{shapeType=Capsule, length=2, radius=0.4, shapeName=dyn_sp05_02_capsule, center=<1.3514 0.1901 0>, rotation=<rot 90 7.125 -9.7501>}]
 dyn_sp06 [{shapeType=Capsule, length=2, radius=0.4, shapeName=dyn_sp06_capsule, center=<1.3865 0 0>, rotation=<rot 90 0 0>}]
 dyn_sp06_02 [{shapeType=Capsule, length=2, radius=0.4, shapeName=dyn_sp06_02_capsule, center=<1.3865 0 0>, rotation=<rot 90 0 0>}]
 dyn_sp06_03 [{
   shapeType=Capsule
   length=1.5683
   radius=0.3853
   shapeName=dyn_sp06_03_capsule
   center=<1.4683 0.373 0.0698>
   rotation=<rot 0 104.0512 -88.6996>
  }]
}
constraints [
 {bone1Name=dyn_box_L, bone2Name=base_box_L, swing1LimitDegrees=1, swing2LimitDegrees=20, twistLimitDegrees=1}
 {bone1Name=dyn_box_R, bone2Name=base_box_R, swing1LimitDegrees=1, swing2LimitDegrees=20, twistLimitDegrees=1}
 {bone1Name=dyn_box_back, bone2Name=base_box_back, swing1LimitDegrees=1, swing2LimitDegrees=20, twistLimitDegrees=1}
 {bone1Name=dyn_box_front, bone2Name=base_box_front, swing1LimitDegrees=1, swing2LimitDegrees=20, twistLimitDegrees=1}
 {bone1Name=dyn_sp01, bone2Name=base_sp01, swing1LimitDegrees=15, swing2LimitDegrees=30, twistLimitDegrees=1}
 {bone1Name=dyn_sp02, bone2Name=base_sp02, swing1LimitDegrees=40, swing2LimitDegrees=5, twistLimitDegrees=1}
 {bone1Name=dyn_sp03, bone2Name=base_sp03, swing1LimitDegrees=50, swing2LimitDegrees=15, twistLimitDegrees=1}
 {bone1Name=dyn_sp03_02, bone2Name=dyn_sp03, swing1LimitDegrees=20, swing2LimitDegrees=25, twistLimitDegrees=1}
 {bone1Name=dyn_sp04, bone2Name=base_sp04, swing1LimitDegrees=45, swing2LimitDegrees=15, twistLimitDegrees=1}
 {bone1Name=dyn_sp04_02, bone2Name=dyn_sp04, swing1LimitDegrees=20, swing2LimitDegrees=20, twistLimitDegrees=1}
 {bone1Name=dyn_sp04_03, bone2Name=dyn_sp04_02, swing1LimitDegrees=40, swing2LimitDegrees=40, twistLimitDegrees=1}
 {bone1Name=dyn_sp06, bone2Name=base_sp06, swing1LimitDegrees=60, swing2LimitDegrees=40, twistLimitDegrees=1}
 {bone1Name=dyn_sp06_02, bone2Name=dyn_sp06, swing1LimitDegrees=40, swing2LimitDegrees=30, twistLimitDegrees=1}
 {bone1Name=dyn_sp06_03, bone2Name=dyn_sp06, swing1LimitDegrees=30, swing2LimitDegrees=35, twistLimitDegrees=1}
 {bone1Name=dyn_sp05, bone2Name=base_sp05, swing1LimitDegrees=30, swing2LimitDegrees=20, twistLimitDegrees=1}
 {bone1Name=dyn_sp05_02, bone2Name=dyn_sp05, swing1LimitDegrees=50, swing2LimitDegrees=30, twistLimitDegrees=1}
 {bone1Name=dyn_meatball02, bone2Name=base_meatball02, swing1LimitDegrees=1, swing2LimitDegrees=1, twistLimitDegrees=1}
 {bone1Name=dyn_meatball01, bone2Name=base_meatball01, swing1LimitDegrees=1, swing2LimitDegrees=1, twistLimitDegrees=1}
]
